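# Translucent water material for the Asteria infinite pool waterfall/fountain, with scrolling normals, foam and broken-up opacity.
# Panned normal and mask textures combine with vertex-color masks to build foam and opacity, with tiling scaled by object X scale.
# Distorted, distance-tinted SceneColor gains switchable skylight reflections, time-of-day foam and custom sun specular.
# Low/mobile paths swap scene color for a lit flat color, low/medium quality drops specular, and HLOD gets a bake color.
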Material /Game/Athena/Asteria/Environments/Water/InfinitePoolWaterfall/Materials/M_InfiniteWaterFountain
graph {
 properties {
  bUsedWithInstancedStaticMeshes=true
  bUsedWithNanite=true
  bUsedWithNiagaraMeshParticles=true
  blendMode=BLEND_Translucent
  pixelDepthOffsetMode=PDOM_Legacy
  translucencyLightingMode=TLM_VolumetricDirectional
  usedShadingModels=Default Lit
 }
 nodes [
  ComponentMask(None=NamedRerouteUsage()) {g=true}
  n1 = NamedRerouteUsage()
  n2 = Constant2Vector() {g=1, r=0.8}
  n3 = Multiply(
   A=TextureCoordinate()
   B=MaterialFunctionCall(X=Multiply(A=@n1, B=@n2.R), Y=@n2.G) {materialFunction=&MakeFloat2}.Result
  )
  n4 = Constant2Vector() {g=-0.25, r=0.05}
  n5 = MaterialFunctionCall(X=@n4, Y=Multiply(A=@n4.G, B=ScalarParameter() {defaultValue=1, parameterName=Water Speed})) {materialFunction=&MakeFloat2}
  n6 = Multiply(A=NamedRerouteUsage()) {constB=0.005}
  n7 = TextureSample(UVs=Add(A=Panner(Coordinate=@n3, Speed=@n5.Result) {speedX=0.05, speedY=-1}, B=@n6)) {samplerType=SAMPLERTYPE_Normal, texture=/Game/Athena/Asteria/Environments/Water/InfinitePoolWaterfall/Textures/T_WaterFountain_Normal}
  n8 = NamedRerouteUsage()
  n9 = Constant2Vector() {g=0.8, r=1}
  n10 = Multiply(
   A=TextureCoordinate()
   B=MaterialFunctionCall(X=Multiply(A=@n8, B=@n9.R), Y=@n9.G) {materialFunction=&MakeFloat2}.Result
  )
  n11 = Constant2Vector() {g=-0.5, r=-0.1}
  n12 = MaterialFunctionCall(X=@n11, Y=Multiply(A=@n11.G, B=ScalarParameter() {defaultValue=1, parameterName=Water Speed})) {materialFunction=&MakeFloat2}
  n13 = TextureSample(UVs=Add(A=Panner(Coordinate=@n10, Speed=@n12.Result) {speedX=0.05, speedY=-1}, B=@n6)) {samplerType=SAMPLERTYPE_Normal, texture=&T_WaterFountain_Normal}
  LinearInterpolate(A=@n7.RGB, B=@n13.RGB, Alpha=SmoothStep(Value=NamedRerouteUsage()) {constMin=0.1})
  MaterialProxyReplace(Realtime=Constant(), MaterialProxy=VectorParameter() {group=HLOD, parameterName=HLODBakeColor, defaultValue=<rgba 0.2705 0.9301 1 1>}.RGB)
  n14 = NamedRerouteUsage()
  n15 = Constant2Vector() {g=0.25, r=0.75}
  n16 = Multiply(
   A=TextureCoordinate()
   B=MaterialFunctionCall(X=Multiply(A=@n14, B=@n15.R), Y=@n15.G) {materialFunction=&MakeFloat2}.Result
  )
  n17 = Constant2Vector() {g=-0.5, r=-0.1}
  n18 = ScalarParameter() {defaultValue=1, parameterName=Water Speed}
  n19 = MaterialFunctionCall(X=@n17, Y=Multiply(A=@n17.G, B=Divide(A=@n18) {constB=5})) {materialFunction=&MakeFloat2}
  n20 = NamedRerouteUsage()
  n21 = TextureSample(UVs=Add(A=Panner(Coordinate=@n16, Speed=@n19.Result) {speedX=0.05, speedY=-1}, B=Multiply(A=@n20) {constB=0.05})) {samplerType=SAMPLERTYPE_Normal, texture=&T_WaterFountain_Normal}
  Multiply(A=Power(Base=@n21.G, Exp=ScalarParameter() {defaultValue=1.5, parameterName=Foam Falling Amount}) {constExponent=1.5}) {constB=3}
  n22 = VertexColor()
  NamedRerouteDeclaration(None=Power(Base=@n22.A) {constExponent=1.5}) {name=Top Foam Mask, nodeColor=<rgba 0 0.4844 1 1>}
  NamedRerouteDeclaration(None=Multiply(A=@n22.B, B=@n22.B)) {name=Bottom Mask, nodeColor=<rgba 1 0 0.6328 1>}
  n23 = TextureCoordinate()
  NamedRerouteDeclaration(None=@n23) {name=UV, nodeColor=<rgba 0.9922 0 1 1>}
  NamedRerouteDeclaration(None=Multiply(A=@n22.R, B=@n22.R)) {name=Top Mask, nodeColor=<rgba 0.8281 0 1 1>}
  NamedRerouteDeclaration(None=@n22.G) {name=Foam Mask, nodeColor=<rgba 0 0.2734 1 1>}
  n24 = VectorParameter() {parameterName=LowEnd Color, defaultValue=<rgba 0.2747 0.9301 1 1>}
  NamedRerouteDeclaration(None=@n24.A) {name=LowEndColor Alpha, nodeColor=<rgba 0.7812 1 0 1>}
  NamedRerouteDeclaration(None=Multiply(
    A=MaterialFunctionCall() {materialFunction=&ObjectScale}.Scale X
    B=ScalarParameter() {defaultValue=1, parameterName=WaterTiling}
   )) {name=Horizontal Tiling, nodeColor=<rgba 1 0.2578 0 1>}
  n25 = TextureCoordinate()
  n26 = Constant2Vector() {g=0.5, r=1}
  n27 = TextureSample(UVs=Panner(Coordinate=Multiply(A=@n25, B=@n26)) {speedX=-0.05, speedY=0.05}) {samplerType=SAMPLERTYPE_LinearColor, texture=/Game/Athena/Asteria/Environments/Water/InfinitePoolWaterfall/Textures/T_WaterFountain_Mask}
  NamedRerouteDeclaration(None=VertexInterpolator(VS=@n27.A).PS) {name=Cloud Noise 2, nodeColor=<rgba 1 0.3047 0 1>}
  n28 = NamedRerouteUsage()
  NamedRerouteDeclaration(None=Abs(None=ComponentMask(None=@n28) {r=true})) {name=Lateral Mask, nodeColor=<rgba 1 0 0.0406 1>}
  n29 = Constant3Vector() {constant=<rgba 0 0 1 0>}
  n30 = MaterialFunctionCall(BaseNormal=@n7.RGB, AdditionalNormal=@n13.RGB) {materialFunction=&BlendAngleCorrectedNormals}
  n31 = Add(A=Multiply(A=NamedRerouteUsage()) {constB=5}, B=Multiply(A=NamedRerouteUsage()) {constB=5})
  n32 = NamedRerouteUsage()
  n33 = OneMinus(None=Saturate(None=Add(A=@n31, B=@n32)))
  NamedRerouteDeclaration(None=LinearInterpolate(A=@n29, B=LinearInterpolate(A=@n29, B=@n30.Result, Alpha=@n33), Alpha=ScalarParameter() {defaultValue=1, parameterName=NormalIntensity})) {name=Normals, nodeColor=<rgba 1 0.8438 0 1>}
  NamedRerouteDeclaration(None=Subtract(A=Multiply(A=@n23) {constB=2})) {name=1-1-UV, nodeColor=<rgba 1 0 0.0469 1>}
  n34 = NamedRerouteUsage()
  n35 = Multiply(A=Multiply(A=OneMinus(None=@n34)) {constB=1.5}, B=NamedRerouteUsage())
  n36 = NamedRerouteUsage()
  n37 = Multiply(A=Saturate(None=Power(Base=@n36) {constExponent=10})) {constB=0.25}
  n38 = Power(Base=@n36, Exp=ScalarParameter() {defaultValue=5, parameterName=Foam Spread}) {constExponent=5}
  n39 = TextureCoordinate()
  n40 = Constant() {r=0.25}
  n41 = TextureSample(UVs=Panner(Coordinate=MaterialFunctionCall(UVs=@n39, "Rotation Angle (0-1)"=@n40) {materialFunction=&CustomRotator}.Rotated Values) {speedX=0.1, speedY=0.1}) {samplerType=SAMPLERTYPE_LinearColor, texture=&T_WaterFountain_Mask}
  n42 = Add(A=@n38, B=Subtract(A=@n36, B=NamedRerouteDeclaration(None=@n41.A) {name=Cloud Noise, nodeColor=<rgba 1 0.859 0 1>}))
  n43 = TextureCoordinate()
  n44 = NamedRerouteUsage()
  n45 = Multiply(A=@n43, B=Multiply(A=@n44) {constB=0.5})
  n46 = TextureSample(UVs=Panner(Coordinate=Multiply(A=@n45) {constB=2}) {speedX=0.1, speedY=-0.75}) {samplerType=SAMPLERTYPE_LinearColor, texture=&T_WaterFountain_Mask}
  n47 = Multiply(A=@n38, B=Multiply(A=Saturate(None=@n42), B=@n46.B))
  n48 = Multiply(A=Saturate(None=InverseLinearInterpolate(Value=@n35) {constB=0.3}), B=Saturate(None=Add(A=@n37, B=@n47)))
  n49 = ScalarParameter() {defaultValue=1, parameterName=Foam Intensity}
  n50 = Multiply(A=@n49) {constB=0.5}
  n51 = MaterialFunctionCall(Value_Morning=@n49, Value_Day=@n49, Value_Evening=@n50, Value_Night=@n50) {materialFunction=&Scalar_Time_of_Day}
  n52 = Add(A=Saturate(None=Multiply(A=@n48, B=@n51.Result)), B=Multiply(A=@n41.R, B=NamedRerouteUsage()))
  n53 = NamedRerouteUsage()
  n54 = ScalarParameter() {defaultValue=1, parameterName=NormalIntensity}
  n55 = Power(Base=ComponentMask(None=Divide(A=@n53, B=@n54)) {r=true}) {constExponent=8}
  n56 = Multiply(A=Saturate(None=Multiply(A=@n46.G, B=@n55))) {constB=0.01}
  n57 = Multiply(A=NamedRerouteUsage()) {constB=0.1}
  n58 = Multiply(A=Power(Base=Subtract(A=@n57, B=@n21.G)) {constExponent=4}) {constB=30}
  n59 = Multiply(A=Saturate(None=Add(A=@n56, B=@n58)), B=ScalarParameter() {defaultValue=1, parameterName=Foam Falling Intensity})
  n60 = Multiply(A=@n59) {constB=0.5}
  n61 = MaterialFunctionCall(Value_Morning=@n59, Value_Day=@n59, Value_Evening=@n60, Value_Night=@n60) {materialFunction=&Scalar_Time_of_Day}
  NamedRerouteDeclaration(None=Saturate(None=Add(A=@n52, B=@n61.Result))) {name=Foam, nodeColor=<rgba 1 0 0.4687 1>}
  n62 = NamedRerouteUsage()
  n63 = Multiply(A=Multiply(A=@n62) {constB=3}, B=NamedRerouteUsage())
  n64 = NamedRerouteUsage()
  n65 = Add(A=Add(A=@n62, B=Multiply(A=@n63, B=@n63)), B=Multiply(A=Multiply(A=@n64, B=@n64)) {constB=2})
  n66 = Multiply(
   A=OneMinus(None=NamedRerouteUsage())
   B=Multiply(
    A=MaterialFunctionCall() {materialFunction=&ObjectScale}.Scale X
    B=ScalarParameter() {defaultValue=1, parameterName=WaterTiling}
   )
  )
  n67 = Multiply(A=@n66) {constB=3}
  n68 = NamedRerouteUsage()
  n69 = Constant2Vector() {g=0.1, r=2}
  n70 = Multiply(
   A=Multiply(A=TextureCoordinate()) {constB=2}
   B=MaterialFunctionCall(X=Multiply(A=@n68, B=@n69.R), Y=@n69.G) {materialFunction=&MakeFloat2}.Result
  ) {constB=1.02}
  n71 = TextureSample(UVs=Panner(Coordinate=@n70) {speedX=-0.1, speedY=-1}) {samplerType=SAMPLERTYPE_LinearColor, texture=&T_WaterFountain_Mask}
  n72 = TextureSample(UVs=Panner(Coordinate=@n70) {speedX=0.1, speedY=-0.75}) {samplerType=SAMPLERTYPE_LinearColor, texture=&T_WaterFountain_Mask}
  n73 = Subtract(A=@n71.G, B=@n72.G)
  n74 = Subtract(A=@n66, B=Subtract(A=@n72.A, B=@n71.A))
  n75 = LinearInterpolate(B=Add(A=Add(A=@n67, B=@n73), B=Saturate(None=@n74)), Alpha=Multiply(A=@n66) {constB=20})
  n76 = Constant2Vector() {g=0.2, r=3}
  n77 = MaterialFunctionCall(X=Multiply(A=@n44, B=@n76.R), Y=@n76.G) {materialFunction=&MakeFloat2}
  n78 = TextureSample(UVs=Panner(Coordinate=Multiply(A=@n43, B=@n77.Result) {constB=1.02}) {speedX=-0.1, speedY=-1}) {samplerType=SAMPLERTYPE_LinearColor, texture=&T_WaterFountain_Mask}
  n79 = Subtract(A=Multiply(A=@n46.G, B=@n78.R), B=Multiply(A=NamedRerouteUsage()) {constB=0.2})
  n80 = SquareRoot(None=NamedRerouteUsage())
  n81 = Multiply(A=Subtract(A=SmoothStep(Value=@n75) {constMax=0.5, constMin=0.05}, B=Saturate(None=@n79)), B=OneMinus(None=Multiply(A=@n80) {constB=20}))
  n82 = MaterialFunctionCall!(Default=Constant() {r=1}, Low/Mobile=NamedRerouteUsage()) {materialFunction=/Game/Effects/Fort_Effects/Materials/_Functions/MF_LowMobile_PathingSwitch}
  n83 = Saturate(None=NamedRerouteUsage())
  n84 = Saturate(None=Add(A=Multiply(A=@n81, B=@n82.Result) {constB=0.2}, B=Multiply(A=@n81, B=@n83)))
  n85 = NamedRerouteUsage()
  NamedRerouteDeclaration(None=Multiply(A=OneMinus(None=@n65), B=LinearInterpolate(A=@n84, Alpha=@n85))) {name=Opacity, nodeColor=<rgba 1 0.6563 0 1>}
  n86 = LinearInterpolate(A=VertexNormalWS(), B=PixelNormalWS(), Alpha=ScalarParameter() {defaultValue=0.5, parameterName=Reflection Distortion Intensity})
  NamedRerouteDeclaration(None=SkyLightEnvMapSample(Direction=ReflectionVectorWS(CustomWorldNormal=@n86), Roughness=ScalarParameter() {defaultValue=0.1, parameterName=Reflection Roughness})) {name=Reflections, nodeColor=<rgba 1 0.2344 0 1>}
  n87 = VectorParameter() {parameterName=Specular Morning Color, defaultValue=<rgba 1 1 1 1>}
  n88 = VectorParameter() {parameterName=Specular Day Color, defaultValue=<rgba 1 1 1 1>}
  n89 = VectorParameter() {parameterName=Specular Evening Color, defaultValue=<rgba 1 1 1 1>}
  n90 = VectorParameter() {parameterName=Specular Night Color, defaultValue=<rgba 1 1 1 1>}
  n91 = MaterialFunctionCall!(Normal=PixelNormalWS(), CameraVector=SkyAtmosphereLightDirection()) {materialFunction=/Engine/Functions/Engine_MaterialFunctions02/Utility/CustomReflectionVector}
  n92 = CameraVectorWS()
  n93 = NamedRerouteUsage()
  n94 = Multiply(A=Saturate(None=DotProduct(A=@n91.Reflection Vector, B=@n92)), B=OneMinus(None=Multiply(A=@n93) {constB=3}))
  n95 = ScalarParameter() {defaultValue=7.6, parameterName=Specular Smoothness}
  n96 = ComponentMask(None=NamedRerouteUsage()) {g=true}
  n97 = Multiply(
   A=Multiply(A=ScalarParameter() {defaultValue=0.3, parameterName=Specular Intensity}, B=Power(Base=@n94, Exp=@n95) {constExponent=3}) {constB=2}
   B=Power(Base=OneMinus(None=@n96)) {constExponent=4}
  )
  NamedRerouteDeclaration(None=Multiply(
    A=MaterialFunctionCall(Color_Morning=@n87.RGB, Color_Day=@n88.RGB, Color_Evening=@n89.RGB, Color_Night=@n90.RGB) {materialFunction=&Color_Time_of_Day}.Result
    B=Saturate(None=@n97)
   )) {name=Specular, nodeColor=<rgba 0.3272 0 1 1>}
  n98 = Constant()
  QualitySwitch(Default=NamedRerouteUsage(), Low=@n98, Medium=@n98)
  ScalarParameter() {defaultValue=0.75, parameterName=Specular Amount}
  n99 = NamedRerouteUsage()
  n100 = Multiply(A=ScalarParameter() {defaultValue=0.1, parameterName=FixReflectionIssueMul}, B=@n99)
  n101 = ScreenPosition()
  n102 = ComponentMask(None=@n101.ViewportUV) {r=true}
  n103 = ComponentMask(None=@n101.ViewportUV) {g=true}
  n104 = MaterialFunctionCall(Float3=NamedRerouteUsage()) {materialFunction=&BreakOutFloat3Components}
  n105 = Add(A=@n104.R, B=@n104.R)
  n106 = Constant() {r=0.03}
  n107 = ScalarParameter() {defaultValue=0.5, parameterName=Distortion Intensity}
  n108 = WorldPosition()
  n109 = CameraPositionWS()
  n110 = Saturate(None=Divide(A=Distance(A=@n108.XYZ, B=@n109)) {constB=3200})
  n111 = Multiply(A=Multiply(A=Add(A=@n105, B=@n104.B), B=Multiply(A=@n106, B=@n107)), B=Clamp(None=OneMinus(None=@n110), Min=@n107)) {constB=0.03}
  n112 = Add(A=@n103, B=@n111)
  n113 = Multiply(
   A=SceneColor!(UVs=AppendVector(A=@n102, B=@n112)).RGB
   B=VectorParameter() {parameterName=Color Tint Near, defaultValue=<rgba 0.2812 0.9329 1 1>}.RGB
  ) {desc=Realtime}
  n114 = VectorParameter() {parameterName=Color Tint Far, defaultValue=<rgba 0.2812 0.9329 1 1>}
  n115 = Divide(A=Distance(A=WorldPosition().XYZ, B=CameraPositionWS())) {constB=30000}
  n116 = PixelNormalWS()
  n117 = SkyAtmosphereLightDirection()
  n118 = Add(A=Saturate(None=DotProduct(A=@n116, B=@n117))) {constB=0.75}
  n119 = MaterialFunctionCall(
   Default=LinearInterpolate(A=@n113, B=LinearInterpolate(A=@n113, B=@n114.RGB, Alpha=@n114.RGB), Alpha=Saturate(None=@n115))
   Low/Mobile=Multiply(A=Saturate(None=@n118), B=@n24.RGB) {desc=Low-End Version}
  ) {materialFunction=&MF_LowMobile_PathingSwitch}
  n120 = LinearInterpolate(A=@n119.Result, B=@n99, Alpha=ScalarParameter() {defaultValue=0.075, parameterName=ReflectionIntensity, sliderMax=0.1}) {constAlpha=0.05}
  n121 = Multiply(A=NamedRerouteUsage(), B=VectorParameter() {parameterName=Foam Color, defaultValue=<rgba 0.6673 0.7401 0.7552 1>}.RGB)
  n122 = Add(
   A=StaticSwitchParameter(True=StaticSwitchParameter(True=@n100, False=@n120) {parameterName=FixReflectionIssue}, False=@n119.Result) {defaultValue=true, parameterName=EnableReflections}
   B=MaterialFunctionCall(Color_Morning=@n121, Color_Day=@n121, Color_Evening=@n121, Color_Night=@n121) {materialFunction=&Color_Time_of_Day}.Result
  )
  n123 = NamedRerouteUsage()
  n124 = Constant()
  StaticSwitchParameter(True=Add(A=@n122, B=QualitySwitch(Default=@n123, Low=@n124, Medium=@n124)), False=@n122) {defaultValue=true, parameterName=EnableSpecular}
  n125 = NamedRerouteUsage()
  StaticSwitchParameter(True=Multiply(A=@n125, B=ScalarParameter() {defaultValue=0.5, parameterName=FixReflectionIssueOpacityMul}), False=@n125) {parameterName=FixReflectionIssue}
  Subtract(A=@n103, B=@n111)
 ]
}
